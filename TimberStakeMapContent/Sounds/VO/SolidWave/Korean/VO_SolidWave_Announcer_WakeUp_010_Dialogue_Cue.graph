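# Sound cue for a Korean Solid Wave announcer "WakeUp" voice-over line in the TimberStake map content.
# A single dialogue player node plays the matching dialogue wave with the announcer as speaker and no targets.
# The cue uses the NPC voice-over sound class and a very large max distance.
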SoundCue /TimberStakeMapContent/Sounds/VO/SolidWave/Korean/VO_SolidWave_Announcer_WakeUp_010_Dialogue_Cue
root=DialoguePlayer_0
nodes [{
  id=DialoguePlayer_0
  properties {dialogueWaveParameter {
    dialogueWave=/TimberStakeMapContent/Sounds/VO/SolidWave/Korean/Wavs/VO_SolidWave_Announcer_WakeUp_010_Dialogue
    context {speaker=/TimberStakeMapContent/Sounds/VO/SolidWave/Korean/Wavs/VO_TimberStake_SW_Announcer, targets []}
   }}
 }]
properties {maxDistance=2097152, soundClassObject=&Fort_VO_NPC, volumeMultiplier=1}
