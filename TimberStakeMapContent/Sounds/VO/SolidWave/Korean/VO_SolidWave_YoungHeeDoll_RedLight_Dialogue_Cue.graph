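# Korean voice-over Sound Cue for the YoungHee doll's Red Light call in SolidWave.
# A Random node at the root picks one of four equally weighted dialogue players.
# Each player plays its own Dialogue Wave with the doll as speaker and no targets, so the line resolves through the dialogue system.
# The cue uses the NPC voice sound class and a very large max distance.
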SoundCue /TimberStakeMapContent/Sounds/VO/SolidWave/Korean/VO_SolidWave_YoungHeeDoll_RedLight_Dialogue_Cue
root=Random_0
nodes [
 {
  id=DialoguePlayer_0
  properties {dialogueWaveParameter {
    dialogueWave=/TimberStakeMapContent/Sounds/VO/SolidWave/Korean/Wavs/VO_SolidWave_YoungHeeDoll_RedLight_010_Dialogue
    context {speaker=/TimberStakeMapContent/Sounds/VO/SolidWave/Korean/Wavs/VO_TimberStake_SW_Doll, targets []}
   }}
 }
 {
  id=DialoguePlayer_1
  properties {dialogueWaveParameter {
    dialogueWave=/TimberStakeMapContent/Sounds/VO/SolidWave/Korean/Wavs/VO_SolidWave_YoungHeeDoll_RedLight_011_Dialogue
    context {speaker=&VO_TimberStake_SW_Doll, targets []}
   }}
 }
 {
  id=DialoguePlayer_2
  properties {dialogueWaveParameter {
    dialogueWave=/TimberStakeMapContent/Sounds/VO/SolidWave/Korean/Wavs/VO_SolidWave_YoungHeeDoll_RedLight_012_Dialogue
    context {speaker=&VO_TimberStake_SW_Doll, targets []}
   }}
 }
 {
  id=DialoguePlayer_3
  properties {dialogueWaveParameter {
    dialogueWave=/TimberStakeMapContent/Sounds/VO/SolidWave/Korean/Wavs/VO_SolidWave_YoungHeeDoll_RedLight_013_Dialogue
    context {speaker=&VO_TimberStake_SW_Doll, targets []}
   }}
 }
 {id=Random_0, children [DialoguePlayer_0, DialoguePlayer_1, DialoguePlayer_2, DialoguePlayer_3], properties {weights [1,1,1,1]}}
]
properties {maxDistance=2097152, soundClassObject=&Fort_VO_NPC, volumeMultiplier=1}
