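# Sound cue for Korean Red Light voice line 013 of the YoungHee doll in TimberStake map content.
# A single Dialogue Player root node plays the line's Dialogue Wave, with the SW Doll voice as speaker and no targets.
# It routes to the Fort_VO_NPC sound class and has a very large max distance.
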SoundCue /TimberStakeMapContent/Sounds/VO/SolidWave/Korean/VO_SolidWave_YoungHeeDoll_RedLight_Dialogue_Cue013
root=DialoguePlayer_3
nodes [{
  id=DialoguePlayer_3
  properties {dialogueWaveParameter {
    dialogueWave=/TimberStakeMapContent/Sounds/VO/SolidWave/Korean/Wavs/VO_SolidWave_YoungHeeDoll_RedLight_013_Dialogue
    context {speaker=/TimberStakeMapContent/Sounds/VO/SolidWave/Korean/Wavs/VO_TimberStake_SW_Doll, targets []}
   }}
 }]
properties {maxDistance=2097152, soundClassObject=&Fort_VO_NPC, volumeMultiplier=1}
